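; Fortnite building prop Blueprint for a CozyTea dark bun table with chairs that recolors its meshes from building texture data.
; Its construction script runs the parent script, then passes the first texture entry's tint to SetColoronSecondaryMesh for the static mesh component.
; When a second entry exists, its tint colors the InstancedBuns component, in a separate array because instanced and static meshes cannot be mixed.
; BeginPlay, overlap and Tick events are empty.
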(event EventBeginPlay)

(event Collision|EventActorBeginOverlap
  (params (OtherActor :type Actor%20Object%20Reference)))

(event EventTick
  (params (DeltaSeconds :type Float%20%28single-precision%29)))

(fn ConstructionScript
  (|CallParentFunction)
  (bind _outtexturedataarray (Building|GetBuildingTextureData))
  (if (> (Utilities|Array|Length _outtexturedataarray) 0)
    (bind _output_get (Variables|Default|SetTint%20Color (Variables|/Script/FortniteGame.BuildingTextureData|GetTintColor (Utilities|Array|Get(acopy) _outtexturedataarray))))
    ;[1 1 1] can't mismatch ISM  and SMs
    (for _array_element (Utilities|Array|MakeArray (Variables|Debug|GetStaticMeshComponent))
      (Default|SetColoronSecondaryMesh (Math|Conversions|ToLinearColor(Color) _output_get) _array_element)
      ;[1 1 1] can't mismatch ISM and SMs
      (if (> (Utilities|Array|Length _outtexturedataarray) 1)
        (bind _output_get_2 (Variables|Default|SetTint%20Color_0 (Variables|/Script/FortniteGame.BuildingTextureData|GetTintColor (Utilities|Array|Get(acopy) _outtexturedataarray 1))))
        (for _array_element_2 (Utilities|Array|MakeArray (Variables|Default|GetInstancedBuns))
          (Default|SetColoronSecondaryMesh (Math|Conversions|ToLinearColor(Color) _output_get_2) _array_element_2))))))
